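# Base widget for Creative EOM screen animation styles, layering Back, UsingColorSet and Front named slots in an overlay.
# Play EOM Animation optionally gathers the images inside the UsingColorSet slot and recolors them by a color-set enum.
# Get Color Set looks up that enum in the DA_Creative_EOMScreens data asset map.
# SetMaterialColors writes the resulting background, primary and accent colors into each image's dynamic material parameters.
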WidgetBlueprint /Game/Athena/HUD/Creative/EOMScreens/Styles/EOMStyles/WBP_Creative_EOMScreens_AnimationBase
graphs [
 EventGraph {nodes [
   n1 = Play EOM Animation!() -> (OutputDelegate, then, bUseColorSet:bool=true, InColorSet:E_Creative_EOMScreens_ColorSet Enum=NewEnumerator0) {is=title K2Node_CustomEvent}
   > Branch(Condition=@n1.bUseColorSet)
   > n2 = GetContainedImages!(execute, InWidget:Widget Object Reference=Get UsingColorSet() -> (UsingColorSet:Named Slot Object Reference)) -> (then, ReturnValue:Array of Image Object References)
   > n3 = Set ColorSetImages(ColorSetImages:Array of Image Object References=@n2) -> (Output_Get:Array of Image Object References)
   > SetColors!(execute, self, InMaterialImagesArray:Array of Image Object References=@n3, InColorSet:E_Creative_EOMScreens_ColorSet Enum=@n1.InColorSet) -> (then)
  ]}
 Get Color Set {nodes [
   n1 = Get Color Set!() -> (then, InColorSet:"E_Creative_EOMScreens_ColorSet Enum (by ref)"=NewEnumerator0) {is=title K2Node_FunctionEntry}
   n2 = Find(
    TargetMap:Map of E_Creative_EOMScreens_ColorSet Enums to Struct Creative EOM Color Data Structures="Get EOMScreens_ColorSet "!(self:PDA Creative EOMScreens Object Reference=Get PrimaryDataAsset() -> (PrimaryDataAsset:PDA Creative EOMScreens Object Reference)) -> ("EOMScreens_ColorSet ":Map of E_Creative_EOMScreens_ColorSet Enums to Struct Creative EOM Color Data Structures=NewEnumerator0)
    Key:"E_Creative_EOMScreens_ColorSet Enum (by ref)"=@n1
   ) -> (Value:Struct Creative EOM Color Data Structure)
   > Return Node!(execute, ColorSet:Struct Creative EOM Color Data Structure=@n2, bIsValid:bool=@n2.ReturnValue)
  ]}
 SetColors {nodes [
   n1 = SetColors!() -> (then, InMaterialImagesArray:Array of Image Object References, InColorSet:E_Creative_EOMScreens_ColorSet Enum=NewEnumerator0) {is=title K2Node_FunctionEntry}
   > n2 = For Each Loop(Array:Array of Image Object References=@n1) -> (Array Element:Image Object Reference)
   > n3 = GetDynamicMaterial(self=@n2)
   > SetMaterialColors!(
    execute
    self
    InCurrentMaterial:Material Instance Dynamic Object Reference=@n3
    InColorSet:E_Creative_EOMScreens_ColorSet Enum=@n1.InColorSet
   ) -> (then)
  ]}
 SetMaterialColors {nodes [
   n1 = SetMaterialColors!() -> (then, InCurrentMaterial:Material Instance Dynamic Object Reference, InColorSet:E_Creative_EOMScreens_ColorSet Enum=NewEnumerator0) {is=title K2Node_FunctionEntry}
   > Set CurrentMaterialInstance(CurrentMaterialInstance:Material Instance Dynamic Object Reference=@n1) -> (Output_Get:Material Instance Dynamic Object Reference)
   > n2 = Get Color Set!(execute, self, InColorSet:"E_Creative_EOMScreens_ColorSet Enum (by ref)"=@n1.InColorSet) -> (
    then
    ColorSet_PrimaryColor_1_4_#1:Linear Color Structure
    ColorSet_PrimaryColor_2_6_#2:Linear Color Structure
    ColorSet_AccentColor_1_8_#3:Linear Color Structure
    ColorSet_AccentColor_2_10_#4:Linear Color Structure
    ColorSet_BackgroundColor_1_12_#5:Linear Color Structure
    ColorSet_BackgroundColor_2_14_#6:Linear Color Structure
    ColorSet_Name_23_#7:Text
    ColorSet_Description_21_#8:Text
    bIsValid:bool
   ) {is=title K2Node_CallFunction}
   > n3 = Sequence!(execute) -> (then_0, then_1, then_2:Exec)
   > Set PrimaryColor_1(PrimaryColor_1:Linear Color Structure=@n2) -> (Output_Get:Linear Color Structure)
   > Set PrimaryColor_2(PrimaryColor_2:Linear Color Structure=@n2.ColorSet_PrimaryColor_2_6_#2) -> (Output_Get:Linear Color Structure)
   > Set AccentColor_1(AccentColor_1:Linear Color Structure=@n2.ColorSet_AccentColor_1_8_#3) -> (Output_Get:Linear Color Structure)
   > Set AccentColor_2(AccentColor_2:Linear Color Structure=@n2.ColorSet_AccentColor_2_10_#4) -> (Output_Get:Linear Color Structure)
   > Set BackgroundColor_1(BackgroundColor_1:Linear Color Structure=@n2.ColorSet_BackgroundColor_1_12_#5) -> (Output_Get:Linear Color Structure)
   > Set BackgroundColor_2(BackgroundColor_2:Linear Color Structure=@n2.ColorSet_BackgroundColor_2_14_#6) -> (Output_Get:Linear Color Structure)
   > Set Name(Name:Text=@n2.ColorSet_Name_23_#7) -> (Output_Get:Text)
   > Set Description(Description:Text=@n2.ColorSet_Description_21_#8) -> (Output_Get:Text)
   n4 = Sequence!(execute=@n3.then_1) -> (then_0, then_1)
   > SetVectorParameterValue(
    self=Get CurrentMaterialInstance() -> (CurrentMaterialInstance:Material Instance Dynamic Object Reference)
    ParameterName=Bg Color
    Value=Get BackgroundColor_1() -> (BackgroundColor_1:Linear Color Structure)
   )
   > SetVectorParameterValue(self=Get CurrentMaterialInstance(), ParameterName=Quad 1 Color, Value=Get BackgroundColor_1())
   > SetVectorParameterValue(self=Get CurrentMaterialInstance(), ParameterName=Quad 2 Color, Value=Get BackgroundColor_1())
   > SetVectorParameterValue(self=Get CurrentMaterialInstance(), ParameterName=Quad 3 Color, Value=Get BackgroundColor_1())
   SetVectorParameterValue(execute=@n4.then_1, self=Get CurrentMaterialInstance(), ParameterName=Color 1, Value=Get BackgroundColor_1())
   > SetVectorParameterValue(self=Get CurrentMaterialInstance(), ParameterName=Color 2, Value=Get BackgroundColor_2() -> (BackgroundColor_2:Linear Color Structure))
   n5 = Sequence(execute=@n3.then_2)
   > SetVectorParameterValue(self=Get CurrentMaterialInstance(), ParameterName=Shape Color 1, Value=Get PrimaryColor_1() -> (PrimaryColor_1:Linear Color Structure))
   > SetVectorParameterValue(self=Get CurrentMaterialInstance(), ParameterName=Shape Color 2, Value=Get PrimaryColor_2() -> (PrimaryColor_2:Linear Color Structure))
   > SetVectorParameterValue(self=Get CurrentMaterialInstance(), ParameterName=Line Color, Value=Get AccentColor_1() -> (AccentColor_1:Linear Color Structure))
   Reroute Node(InputPin=@n5.then_1)
  ]}
]
variables [
 ColorSetImages {type="array<Image>", default []}
 PrimaryDataAsset {type=PDA_Creative_EOMScreens_C, default=/Game/Athena/HUD/Creative/EOMScreens/DataAssets/DA_Creative_EOMScreens}
 Struct_Color {
  type=Struct_Creative_EOM_ColorData
  default {
   name=""
   description=""
   accentColor_1=<rgba 0 0 0 0>
   accentColor_2=<rgba 0 0 0 0>
   backgroundColor_1=<rgba 0 0 0 0>
   backgroundColor_2=<rgba 0 0 0 0>
   primaryColor_1=<rgba 0 0 0 0>
   primaryColor_2=<rgba 0 0 0 0>
  }
 }
]
widgets {
 info {inheritedWidgetCount=0, parentClass=&UserWidget, rootWidgetClass=&Overlay, widgetCount=4}
 widgets [
  Root {widgetClassPath=&Overlay, widget=Root}
  Back {
   widgetClassPath=&NamedSlot
   bIsVariable=true
   parent=Root
   widget=Back
   slot {class=OverlaySlot, properties {horizontalAlignment=HAlign_Fill, verticalAlignment=VAlign_Fill}}
  }
  UsingColorSet {
   widgetClassPath=&NamedSlot
   bIsVariable=true
   parent=Root
   widget=UsingColorSet
   slot {class=OverlaySlot, properties {horizontalAlignment=HAlign_Fill, verticalAlignment=VAlign_Fill}}
  }
  Front {
   widgetClassPath=&NamedSlot
   bIsVariable=true
   parent=Root
   widget=Front
   slot {class=OverlaySlot, properties {horizontalAlignment=HAlign_Fill, verticalAlignment=VAlign_Fill}}
  }
 ]
}
