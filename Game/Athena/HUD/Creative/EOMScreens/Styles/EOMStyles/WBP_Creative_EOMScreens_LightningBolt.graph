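# Lightning-bolt style widget for Creative EOM screens, built on the shared AnimationBase from a background and four material-driven polygon images.
# Play EOM Animation calls the parent with color sets forced on, stops running animations, then chains two intro steps through Finished events.
# All motion animates material parameters, ending in a looping gradient shine on Shape 1.
# The Anim_LineShine play node has no incoming exec link, so this graph never starts it.
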WidgetBlueprint /Game/Athena/HUD/Creative/EOMScreens/Styles/EOMStyles/WBP_Creative_EOMScreens_LightningBolt
animations [
 Anim_LightningBolt_In_Step_1 {
  binding_count=5
  length_seconds=2.5
  object=&self:Anim_LightningBolt_In_Step_1
  sequence {bindings [
    Image_Shape_1 {
     id=#1
     possessed_class=Image
     tracks [{
       type=MovieSceneWidgetMaterialTrack
       sections [{
         type=MovieSceneParameterSection
         end_frame=24
         start_frame=-1
         channels [
          Shape Progress {type=MovieSceneScriptingFloatChannel, keys [[7,0],[20,1]]}
          Line Progress {type=MovieSceneScriptingFloatChannel, keys [[7,0],[20.475,1]]}
          Shape Offset X {type=MovieSceneScriptingFloatChannel, keys [[7,-0.05],[17,0.0268],[22.001,0]]}
          Shape Offset Y {type=MovieSceneScriptingFloatChannel, keys [[7,-0.1],[17,0.0138],[22.001,0]]}
          Line Offset X {type=MovieSceneScriptingFloatChannel, keys [[8.2227,0.13],[18.2227,0.2068],[23.2237,0.18]]}
          Line Offset Y {type=MovieSceneScriptingFloatChannel, keys [[8.2227,-0.31],[18.2227,-0.1962],[23.2237,-0.21]]}
         ]
        }]
      }]
    }
    Image_Shape_2 {
     id=#2
     possessed_class=Image
     tracks [{
       type=MovieSceneWidgetMaterialTrack
       sections [{
         type=MovieSceneParameterSection
         end_frame=26
         start_frame=-1
         channels [
          Shape Progress {type=MovieSceneScriptingFloatChannel, keys [[9,0],[22,1]]}
          Line Progress {type=MovieSceneScriptingFloatChannel, keys [[9,0],[22,1]]}
          Shape Offset X {type=MovieSceneScriptingFloatChannel, keys [[10,-0.2],[17,-0.0485],[23,0]]}
          Shape Offset Y {type=MovieSceneScriptingFloatChannel, keys [[10,0.1],[18,-0.0039],[23,0]]}
          Line Offset X {type=MovieSceneScriptingFloatChannel, keys [[12,-0.2],[18.2227,-0.0354],[25,0]]}
          Line Offset Y {type=MovieSceneScriptingFloatChannel, keys [[12,0.112],[18.2227,0.0636],[25,0.062]]}
          Shape Opacity {type=MovieSceneScriptingFloatChannel, const [1,1]}
          Line Opacity {type=MovieSceneScriptingFloatChannel, const [1,1]}
         ]
        }]
      }]
    }
    Image_Shape_3 {
     id=#3
     possessed_class=Image
     tracks [{
       type=MovieSceneWidgetMaterialTrack
       sections [{
         type=MovieSceneParameterSection
         end_frame=21
         start_frame=-1
         channels [
          Shape Progress {type=MovieSceneScriptingFloatChannel, keys [[6,0],[20,1]]}
          Shape Offset X {type=MovieSceneScriptingFloatChannel, keys [[6,0.2],[14,-0.0103],[20,0]]}
          Shape Offset Y {type=MovieSceneScriptingFloatChannel, keys [[6,-0.1],[14,-0.0056],[20,0]]}
         ]
        }]
      }]
    }
    Image_Shape_4 {
     id=#4
     possessed_class=Image
     tracks [{
       type=MovieSceneWidgetMaterialTrack
       sections [{
         type=MovieSceneParameterSection
         end_frame=20
         start_frame=-1
         channels [
          Shape Progress {type=MovieSceneScriptingFloatChannel, keys [[7,0],[19,1]]}
          Shape Offset X {type=MovieSceneScriptingFloatChannel, const [2.1,3]}
          Shape Offset Y {type=MovieSceneScriptingFloatChannel, keys [[7,-0.6],[15,-0.6292],[19,-0.6]]}
          Shape Opacity {type=MovieSceneScriptingFloatChannel, const [1,1]}
         ]
        }]
      }]
    }
    Image_Background {
     id=#5
     possessed_class=Image
     tracks [{
       type=MovieSceneWidgetMaterialTrack
       sections [{
         type=MovieSceneParameterSection
         end_frame=50
         start_frame=-1
         channels [
          Quad 1 Progress {type=MovieSceneScriptingFloatChannel, keys [[2.9897,0],[13.455,1]]}
          Quad 2 Progress {type=MovieSceneScriptingFloatChannel, keys [[-0.0003,0],[8.2227,1]]}
          Quad 3 Progress {type=MovieSceneScriptingFloatChannel, keys [[2.242,0],[12.7077,1]]}
          Bg Progress {type=MovieSceneScriptingFloatChannel, keys [[7.475,0],[14.95,1]]}
          Bg Progress Rotation {type=MovieSceneScriptingFloatChannel, const [0.5,1]}
         ]
        }]
      }]
    }
   ]}
 }
 Anim_LightningBolt_In_Step_2 {
  binding_count=5
  length_seconds=0.7
  object=&self:Anim_LightningBolt_In_Step_2
  sequence {bindings [
    Image_Background {
     id=#6
     possessed_class=Image
     tracks [{
       type=MovieSceneWidgetMaterialTrack
       sections [{
         type=MovieSceneParameterSection
         end_frame=14
         start_frame=0
         channels [
          Quad 1 Progress {type=MovieSceneScriptingFloatChannel, const [0,1]}
          Quad 2 Progress {type=MovieSceneScriptingFloatChannel, const [0,1]}
          Quad 3 Progress {type=MovieSceneScriptingFloatChannel, const [0,1]}
          Bg Progress {type=MovieSceneScriptingFloatChannel, keys [[0,1],[8,0.5]]}
          Bg Progress Rotation {type=MovieSceneScriptingFloatChannel, const [0.75,1]}
         ]
        }]
      }]
    }
    Image_Shape_3 {
     id=#7
     possessed_class=Image
     tracks [{
       type=MovieSceneWidgetMaterialTrack
       sections [{
         type=MovieSceneParameterSection
         end_frame=12
         start_frame=0
         channels [
          Shape Offset X {type=MovieSceneScriptingFloatChannel, keys [[3,0],[6,0.436],[9,0.4719],[12,0.47]]}
          Shape Offset Y {type=MovieSceneScriptingFloatChannel, keys [[3,0],[6,-0.2883],[9,-0.2833],[12,-0.28]]}
         ]
        }]
      }]
    }
    Image_Shape_2 {
     id=#8
     possessed_class=Image
     tracks [{
       type=MovieSceneWidgetMaterialTrack
       sections [{
         type=MovieSceneParameterSection
         end_frame=8
         start_frame=0
         channels [
          Shape Offset X {type=MovieSceneScriptingFloatChannel, keys [[2,0],[8,-0.59]]}
          Shape Offset Y {type=MovieSceneScriptingFloatChannel, keys [[2,0],[8,5.65]]}
          Line Offset X {type=MovieSceneScriptingFloatChannel, keys [[2,0],[8,-0.32]]}
          Line Offset Y {type=MovieSceneScriptingFloatChannel, keys [[2,0.062],[8,0.742]]}
          Shape Opacity {type=MovieSceneScriptingFloatChannel, keys [[6,1],[8,0]]}
          Line Opacity {type=MovieSceneScriptingFloatChannel, keys [[6,1],[8,0]]}
         ]
        }]
      }]
    }
    Image_Shape_1 {
     id=#9
     possessed_class=Image
     tracks [{
       type=MovieSceneWidgetMaterialTrack
       sections [{
         type=MovieSceneParameterSection
         end_frame=12
         start_frame=0
         channels [
          Shape Offset X {type=MovieSceneScriptingFloatChannel, keys [[2,0],[4,0.0478],[6,0.0645],[8,0.06]]}
          Shape Offset Y {type=MovieSceneScriptingFloatChannel, keys [[2,0],[4,-0.2919],[6,-0.2868],[8,-0.3]]}
          Line Offset X {type=MovieSceneScriptingFloatChannel, const [0.18,2]}
          Line Offset Y {type=MovieSceneScriptingFloatChannel, keys [[3,-0.21],[5,-0.3987],[9,-0.4104],[12,-0.43]]}
         ]
        }]
      }]
    }
    Image_Shape_4 {
     id=#10
     possessed_class=Image
     tracks [{
       type=MovieSceneWidgetMaterialTrack
       sections [{
         type=MovieSceneParameterSection
         end_frame=7
         start_frame=0
         channels [
          Shape Offset X {type=MovieSceneScriptingFloatChannel, keys [[4,2.1],[7,2.9725]]}
          Shape Offset Y {type=MovieSceneScriptingFloatChannel, keys [[4,-0.6],[7,-1.2694]]}
          Shape Opacity {type=MovieSceneScriptingFloatChannel, keys [[6,1],[7,0]]}
         ]
        }]
      }]
    }
   ]}
 }
 Anim_LineShine {
  binding_count=1
  length_seconds=4.9916
  object=&self:Anim_LineShine
  sequence {bindings [Image_Shape_1 {
     id=#11
     possessed_class=Image
     tracks [{
       type=MovieSceneWidgetMaterialTrack
       sections [{
         type=MovieSceneParameterSection
         end_frame=99
         start_frame=0
         channels [
          Line Thickness {type=MovieSceneScriptingFloatChannel, keys [[0,0.0015],[25,0.001],[99.8323,0.0015]]}
          Line ColorR {type=MovieSceneScriptingFloatChannel, const [1,3]}
          Line ColorG {type=MovieSceneScriptingFloatChannel, keys [[0,0.3813],[25,0.65],[99.8323,0.3813]]}
          Line ColorB {type=MovieSceneScriptingFloatChannel, keys [[0,0.0123],[25,0],[99.8323,0.0123]]}
          Line ColorA {type=MovieSceneScriptingFloatChannel, const [1,3]}
         ]
        }]
      }]
    }]}
 }
 Anim_ShapeShine {
  binding_count=1
  length_seconds=5.9927
  object=&self:Anim_ShapeShine
  sequence {bindings [Image_Shape_1 {
     id=#12
     possessed_class=Image
     tracks [{
       type=MovieSceneWidgetMaterialTrack
       sections [{
         type=MovieSceneParameterSection
         end_frame=119
         start_frame=0
         channels [
          Linear Gradient Rotation {type=MovieSceneScriptingFloatChannel, keys [[0,0.0853],[119.853,1.0853]]}
          Gradient Start {type=MovieSceneScriptingFloatChannel, keys [[0,0],[39,0.7959],[68,0.0122],[119.853,0]]}
          Gradient End {type=MovieSceneScriptingFloatChannel, keys [[0,1],[39,1.1606],[90,0.2731],[119.853,1]]}
         ]
        }]
      }]
    }]}
 }
]
graphs [EventGraph {nodes [
   n1 = Event Play EOM Animation!() -> (OutputDelegate, then, bUseColorSet:bool, InColorSet:E_Creative_EOMScreens_ColorSet Enum=NewEnumerator0) {is=title K2Node_Event}
   > Parent: Play EOM Animation!(execute, bUseColorSet:bool=true, InColorSet:E_Creative_EOMScreens_ColorSet Enum=@n1.InColorSet) -> (then) {is=title K2Node_CallParentFunction}
   > StopAllAnimations(self:User Widget Object Reference)
   > n2 = "Play Animation with Finished event (legacy)"!(
    execute:Exec
    Widget:User Widget Object Reference=Self-Reference()
    InAnimation:Widget Animation Object Reference=Get Anim_LightningBolt_In_Step_1() -> (Anim_LightningBolt_In_Step_1:Widget Animation Object Reference)
    StartAtTime:float=0
    NumLoopsToPlay:int=1
    PlayMode:EUMGSequencePlayMode Enum
    PlaybackSpeed:float=1
   ) -> (then:Exec, Result:UMGSequence Player Object Reference, Finished:Exec) {is=title K2Node_PlayAnimation}
   n3 = "Play Animation with Finished event (legacy)"(
    execute=@n2.Finished
    Widget=Self-Reference()
    InAnimation=Get Anim_LightningBolt_In_Step_2() -> (Anim_LightningBolt_In_Step_2:Widget Animation Object Reference)
    StartAtTime=0
    NumLoopsToPlay=1
    PlaybackSpeed=1
   )
   PlayAnimation(
    execute=@n3.Finished
    InAnimation=Get Anim_ShapeShine() -> (Anim_ShapeShine:Widget Animation Object Reference)
    NumLoopsToPlay=0
    PlaybackSpeed=0.75
    bRestoreState=true
   )
   PlayAnimation(
    InAnimation=Get Anim_LineShine() -> (Anim_LineShine:Widget Animation Object Reference)
    NumLoopsToPlay=0
    PlaybackSpeed=0.5
    bRestoreState=true
   )
  ]}]
variables []
widgets {
 info {
  inheritedWidgetCount=0
  namedSlotCount=1
  parentClass=&/WBP_Creative_EOMScreens_AnimationBase.WBP_Creative_EOMScreens_AnimationBase_C
  rootWidgetClass=null
  widgetCount=6
 }
 widgets [
  Content {widgetClassPath=&Overlay, widget=Content}
  Image_Background {
   widgetClassPath=&Image
   bIsVariable=true
   parent=Content
   widget=Image_Background
   properties {
    visibility=HitTestInvisible
    brush {
     bIsDynamicallyLoaded=false
     drawAs=Image
     imageType=NoImage
     mirroring=NoMirror
     resourceName=None
     resourceObject=/Game/Athena/HUD/Creative/EOMScreens/MaterialInstances/MI_UI_Creative_EOMScreens_LightningBolt_Background_3Quads
     tiling=NoTile
     imageSize=<32 32>
     margin=<ltrb 0 0 0 0>
     outlineSettings {
      bUseBrushTransparency=false
      roundingType=HalfHeightRadius
      width=0
      color {colorUseRule=UseColor_Specified, specifiedColor=<rgba 0 0 0 0>}
      cornerRadii=<0 0 0 0>
     }
     tintColor {colorUseRule=UseColor_Specified, specifiedColor=<rgba 1 1 1 1>}
     uVRegion {bIsValid=false, max=<0 0>, min=<0 0>}
    }
   }
   slot {class=OverlaySlot, properties {horizontalAlignment=HAlign_Fill, verticalAlignment=VAlign_Fill}}
  }
  Image_Shape_1 {
   widgetClassPath=&Image
   bIsVariable=true
   parent=Content
   widget=Image_Shape_1
   properties {
    visibility=HitTestInvisible
    brush {
     bIsDynamicallyLoaded=false
     drawAs=Image
     imageType=NoImage
     mirroring=NoMirror
     resourceName=None
     resourceObject=/Game/Athena/HUD/Creative/EOMScreens/MaterialInstances/MI_UI_Creative_EOMScreens_LightningBolt_Polygon_1_6Vertices
     tiling=NoTile
     imageSize=<1100 1100>
     margin=<ltrb 0 0 0 0>
     outlineSettings {
      bUseBrushTransparency=false
      roundingType=HalfHeightRadius
      width=0
      color {colorUseRule=UseColor_Specified, specifiedColor=<rgba 0 0 0 0>}
      cornerRadii=<0 0 0 0>
     }
     tintColor {colorUseRule=UseColor_Specified, specifiedColor=<rgba 1 1 1 1>}
     uVRegion {bIsValid=false, max=<0 0>, min=<0 0>}
    }
   }
   slot {class=OverlaySlot, properties {horizontalAlignment=HAlign_Fill, verticalAlignment=VAlign_Fill}}
  }
  Image_Shape_2 {
   widgetClassPath=&Image
   bIsVariable=true
   parent=Content
   widget=Image_Shape_2
   properties {
    visibility=HitTestInvisible
    brush {
     bIsDynamicallyLoaded=false
     drawAs=Image
     imageType=NoImage
     mirroring=NoMirror
     resourceName=None
     resourceObject=/Game/Athena/HUD/Creative/EOMScreens/MaterialInstances/MI_UI_Creative_EOMScreens_LightningBolt_Polygon_2_6Vertices
     tiling=NoTile
     imageSize=<1100 1100>
     margin=<ltrb 0 0 0 0>
     outlineSettings {
      bUseBrushTransparency=false
      roundingType=HalfHeightRadius
      width=0
      color {colorUseRule=UseColor_Specified, specifiedColor=<rgba 0 0 0 0>}
      cornerRadii=<0 0 0 0>
     }
     tintColor {colorUseRule=UseColor_Specified, specifiedColor=<rgba 1 1 1 1>}
     uVRegion {bIsValid=false, max=<0 0>, min=<0 0>}
    }
   }
   slot {class=OverlaySlot, properties {horizontalAlignment=HAlign_Fill, verticalAlignment=VAlign_Fill}}
  }
  Image_Shape_3 {
   widgetClassPath=&Image
   bIsVariable=true
   parent=Content
   widget=Image_Shape_3
   properties {
    visibility=HitTestInvisible
    brush {
     bIsDynamicallyLoaded=false
     drawAs=Image
     imageType=NoImage
     mirroring=NoMirror
     resourceName=None
     resourceObject=/Game/Athena/HUD/Creative/EOMScreens/MaterialInstances/MI_UI_Creative_EOMScreens_LightningBolt_Polygon_3_3Vertices
     tiling=NoTile
     imageSize=<1100 1100>
     margin=<ltrb 0 0 0 0>
     outlineSettings {
      bUseBrushTransparency=false
      roundingType=HalfHeightRadius
      width=0
      color {colorUseRule=UseColor_Specified, specifiedColor=<rgba 0 0 0 0>}
      cornerRadii=<0 0 0 0>
     }
     tintColor {colorUseRule=UseColor_Specified, specifiedColor=<rgba 1 1 1 1>}
     uVRegion {bIsValid=false, max=<0 0>, min=<0 0>}
    }
   }
   slot {class=OverlaySlot, properties {horizontalAlignment=HAlign_Fill, verticalAlignment=VAlign_Fill}}
  }
  Image_Shape_4 {
   widgetClassPath=&Image
   bIsVariable=true
   parent=Content
   widget=Image_Shape_4
   properties {
    visibility=HitTestInvisible
    brush {
     bIsDynamicallyLoaded=false
     drawAs=Image
     imageType=NoImage
     mirroring=NoMirror
     resourceName=None
     resourceObject=/Game/Athena/HUD/Creative/EOMScreens/MaterialInstances/MI_UI_Creative_EOMScreens_LightningBolt_Polygon_4_3Vertices
     tiling=NoTile
     imageSize=<1100 1100>
     margin=<ltrb 0 0 0 0>
     outlineSettings {
      bUseBrushTransparency=false
      roundingType=HalfHeightRadius
      width=0
      color {colorUseRule=UseColor_Specified, specifiedColor=<rgba 0 0 0 0>}
      cornerRadii=<0 0 0 0>
     }
     tintColor {colorUseRule=UseColor_Specified, specifiedColor=<rgba 1 1 1 1>}
     uVRegion {bIsValid=false, max=<0 0>, min=<0 0>}
    }
   }
   slot {class=OverlaySlot, properties {horizontalAlignment=HAlign_Fill, verticalAlignment=VAlign_Fill}}
  }
 ]
}
