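# Shards-style animated screen for Creative EOM screens, a child of WBP_Creative_EOMScreens_AnimationBase that fills its Content and FrontContent slots.
# Content holds background and shape images; FrontContent holds particle and glow images that start collapsed.
# Play EOM Animation calls the parent, stops all animations, plays In_Step_1, then In_Step_2 when it finishes.
# Both animations drive material parameters: Step 1 builds shapes and bursts particles, Step 2 collapses particles and rescales shapes.
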WidgetBlueprint /Game/Athena/HUD/Creative/EOMScreens/Styles/EOMStyles/WBP_Creative_EOMScreens_Shards
animations [
 Anim_Shards_In_Step_1 {
  binding_count=4
  length_seconds=2.5
  object=&self:Anim_Shards_In_Step_1
  sequence {bindings [
    Image_Particles_Glow {
     id=#1
     possessed_class=Image
     tracks [
      {
       type=MovieSceneWidgetMaterialTrack
       sections [{
         type=MovieSceneParameterSection
         end_frame=31
         start_frame=12
         channels [
          R or G {type=MovieSceneScriptingFloatChannel, keys [[12.0003,0],[30.0003,0.7]]}
          Scale X {type=MovieSceneScriptingFloatChannel, keys [[12,1.2],[30.0003,1]]}
          Scale Y {type=MovieSceneScriptingFloatChannel, keys [[12,1.2],[30.0003,1]]}
          Shape Opacity {type=MovieSceneScriptingFloatChannel, keys [[12.0007,0],[13.0007,1],[30,0]]}
         ]
        }]
      }
      {
       type=MovieSceneByteTrack
       display_name=Visibility
       sections [{type=MovieSceneByteSection, end_frame=31, start_frame=13, channels [None {type=MovieSceneScriptingByteChannel, const [3,1]}]}]
      }
     ]
    }
    Image_Particles {
     id=#2
     possessed_class=Image
     tracks [
      {
       type=MovieSceneWidgetMaterialTrack
       sections [{
         type=MovieSceneParameterSection
         end_frame=26
         start_frame=12
         channels [
          R or G {type=MovieSceneScriptingFloatChannel, keys [[12.0003,0],[25.0003,0.7]]}
          Scale X {type=MovieSceneScriptingFloatChannel, keys [[12,1.2],[25.0003,1]]}
          Scale Y {type=MovieSceneScriptingFloatChannel, keys [[12,1.2],[25.0003,1]]}
          Shape Opacity {type=MovieSceneScriptingFloatChannel, keys [[12.0003,0],[13.0003,1],[25.0003,0]]}
          Progress Rotation {type=MovieSceneScriptingFloatChannel, const [0,1]}
         ]
        }]
      }
      {
       type=MovieSceneByteTrack
       display_name=Visibility
       sections [{type=MovieSceneByteSection, end_frame=26, start_frame=13, channels [None {type=MovieSceneScriptingByteChannel, const [3,1]}]}]
      }
     ]
    }
    Image_Shapes {
     id=#3
     possessed_class=Image
     tracks [{
       type=MovieSceneWidgetMaterialTrack
       sections [{
         type=MovieSceneParameterSection
         end_frame=21
         start_frame=-1
         channels [
          Scale X {type=MovieSceneScriptingFloatChannel, keys [[8.9993,2],[14,1.0029],[15,0.9995],[16,0.9988],[20.0003,1.0019]]}
          Scale Y {type=MovieSceneScriptingFloatChannel, keys [[9,2],[14,1.0026],[15,0.999],[16,0.9981],[20.0007,1]]}
          R or G {type=MovieSceneScriptingFloatChannel, keys [[10.9997,1],[13.9993,0]]}
          Line Scale X {type=MovieSceneScriptingFloatChannel, keys [[8.9997,0.5],[16.9997,1]]}
          Line Scale Y {type=MovieSceneScriptingFloatChannel, keys [[8.9997,0.5],[16.9997,1]]}
          or B Line {type=MovieSceneScriptingFloatChannel, keys [[8.9997,0],[11,0.4767],[15,1]]}
          Line Thickness {type=MovieSceneScriptingFloatChannel, keys [[9.9997,0],[16,0.05]]}
          or A {type=MovieSceneScriptingFloatChannel, const [0,1]}
          Line Offset Y {type=MovieSceneScriptingFloatChannel, const [0,1]}
         ]
        }]
      }]
    }
    Image_Background {
     id=#4
     possessed_class=Image
     tracks [{
       type=MovieSceneWidgetMaterialTrack
       sections [{
         type=MovieSceneParameterSection
         end_frame=50
         start_frame=-1
         channels [
          Shapes Progress {type=MovieSceneScriptingFloatChannel, keys [[0,0],[10,1]]}
          Background Progress {type=MovieSceneScriptingFloatChannel, keys [[8,0],[14,1]]}
          Use Full Background {type=MovieSceneScriptingFloatChannel, const [0,1]}
          Background Rotation {type=MovieSceneScriptingFloatChannel, const [0,1]}
         ]
        }]
      }]
    }
   ]}
 }
 Anim_Shards_In_Step_2 {
  binding_count=4
  length_seconds=1.3
  object=&self:Anim_Shards_In_Step_2
  sequence {bindings [
    Image_Particles_Glow {
     id=#5
     possessed_class=Image
     tracks [{
       type=MovieSceneByteTrack
       display_name=Visibility
       sections [{type=MovieSceneByteSection, end_frame=0, start_frame=0, channels [None {type=MovieSceneScriptingByteChannel, const [1,1]}]}]
      }]
    }
    Image_Particles {
     id=#6
     possessed_class=Image
     tracks [{
       type=MovieSceneByteTrack
       display_name=Visibility
       sections [{type=MovieSceneByteSection, end_frame=1, start_frame=1, channels [None {type=MovieSceneScriptingByteChannel, const [1,1]}]}]
      }]
    }
    Image_Shapes {
     id=#3
     possessed_class=Image
     tracks [{
       type=MovieSceneWidgetMaterialTrack
       sections [{
         type=MovieSceneParameterSection
         end_frame=26
         start_frame=0
         channels [
          Scale X {type=MovieSceneScriptingFloatChannel, keys [[0,1],[8,1.2227],[13,1.18]]}
          Scale Y {type=MovieSceneScriptingFloatChannel, keys [[0,1],[8,1.1869],[14,1.06]]}
          R or G {type=MovieSceneScriptingFloatChannel, const [0,1]}
          Line Scale X {type=MovieSceneScriptingFloatChannel, const [1,1]}
          Line Scale Y {type=MovieSceneScriptingFloatChannel, const [1,1]}
          or B Line {type=MovieSceneScriptingFloatChannel, const [1,1]}
          or A {type=MovieSceneScriptingFloatChannel, keys [[0,0],[12,1]]}
          Line Offset Y {type=MovieSceneScriptingFloatChannel, keys [[0,0],[9,0.0315],[11,0.033],[13,0.03]]}
          Line Thickness {type=MovieSceneScriptingFloatChannel, const [0.05,1]}
         ]
        }]
      }]
    }
    Image_Background {
     id=#4
     possessed_class=Image
     tracks [{
       type=MovieSceneWidgetMaterialTrack
       sections [{
         type=MovieSceneParameterSection
         end_frame=26
         start_frame=0
         channels [
          Background Progress {type=MovieSceneScriptingFloatChannel, keys [[0,1],[7,0.5]]}
          Use Full Background {type=MovieSceneScriptingFloatChannel, const [1,2]}
          Shapes Progress {type=MovieSceneScriptingFloatChannel, const [0,1]}
          Background Rotation {type=MovieSceneScriptingFloatChannel, const [0.75,1]}
         ]
        }]
      }]
    }
   ]}
 }
]
graphs [EventGraph {nodes [
   n1 = Event Play EOM Animation!() -> (OutputDelegate, then, bUseColorSet:bool, InColorSet:E_Creative_EOMScreens_ColorSet Enum=NewEnumerator0) {is=title K2Node_Event}
   > Parent: Play EOM Animation!(execute, bUseColorSet:bool=true, InColorSet:E_Creative_EOMScreens_ColorSet Enum=@n1.InColorSet) -> (then) {is=title K2Node_CallParentFunction}
   > StopAllAnimations(self:User Widget Object Reference)
   > n2 = "Play Animation with Finished event (legacy)"!(
    execute:Exec
    Widget:User Widget Object Reference=Self-Reference()
    InAnimation:Widget Animation Object Reference=Get Anim_Shards_In_Step_1() -> (Anim_Shards_In_Step_1:Widget Animation Object Reference)
    StartAtTime:float=0
    NumLoopsToPlay:int=1
    PlayMode:EUMGSequencePlayMode Enum
    PlaybackSpeed:float=1
   ) -> (then:Exec, Result:UMGSequence Player Object Reference, Finished:Exec) {is=title K2Node_PlayAnimation}
   "Play Animation with Finished event (legacy)"(
    execute=@n2.Finished
    Widget=Self-Reference()
    InAnimation=Get Anim_Shards_In_Step_2() -> (Anim_Shards_In_Step_2:Widget Animation Object Reference)
    StartAtTime=0
    NumLoopsToPlay=1
    PlaybackSpeed=1
   )
  ]}]
variables []
widgets {
 info {
  inheritedWidgetCount=0
  namedSlotCount=2
  parentClass=&/WBP_Creative_EOMScreens_AnimationBase.WBP_Creative_EOMScreens_AnimationBase_C
  rootWidgetClass=null
  widgetCount=6
 }
 widgets [
  Content {widgetClassPath=&Overlay, widget=Content}
  Image_Background {
   widgetClassPath=&Image
   bIsVariable=true
   parent=Content
   widget=Image_Background
   properties {
    renderOpacity=0.8
    brush {
     bIsDynamicallyLoaded=false
     drawAs=Image
     imageType=NoImage
     mirroring=NoMirror
     resourceName=None
     resourceObject=/Game/Athena/HUD/Creative/EOMScreens/MaterialInstances/MI_UI_Creative_EOMScreens_Shards_Backgound_Texture
     tiling=NoTile
     imageSize=<32 32>
     margin=<ltrb 0 0 0 0>
     outlineSettings {
      bUseBrushTransparency=false
      roundingType=HalfHeightRadius
      width=0
      color {colorUseRule=UseColor_Specified, specifiedColor=<rgba 0 0 0 0>}
      cornerRadii=<0 0 0 0>
     }
     tintColor {colorUseRule=UseColor_Specified, specifiedColor=<rgba 1 1 1 1>}
     uVRegion {bIsValid=false, max=<0 0>, min=<0 0>}
    }
   }
   slot {class=OverlaySlot, properties {horizontalAlignment=HAlign_Fill, verticalAlignment=VAlign_Fill}}
  }
  Image_Shapes {
   widgetClassPath=&Image
   bIsVariable=true
   parent=Content
   widget=Image_Shapes
   properties {brush {
     bIsDynamicallyLoaded=false
     drawAs=Image
     imageType=NoImage
     mirroring=NoMirror
     resourceName=None
     resourceObject=/Game/Athena/HUD/Creative/EOMScreens/MaterialInstances/MI_UI_Creative_EOMScreens_Shards_ShapesTexture_1
     tiling=NoTile
     imageSize=<32 32>
     margin=<ltrb 0 0 0 0>
     outlineSettings {
      bUseBrushTransparency=false
      roundingType=HalfHeightRadius
      width=0
      color {colorUseRule=UseColor_Specified, specifiedColor=<rgba 0 0 0 0>}
      cornerRadii=<0 0 0 0>
     }
     tintColor {colorUseRule=UseColor_Specified, specifiedColor=<rgba 1 1 1 1>}
     uVRegion {bIsValid=false, max=<0 0>, min=<0 0>}
    }}
   slot {class=OverlaySlot, properties {horizontalAlignment=HAlign_Fill, verticalAlignment=VAlign_Fill}}
  }
  FrontContent {widgetClassPath=&Overlay, widget=FrontContent}
  Image_Particles_Glow {
   widgetClassPath=&Image
   bIsVariable=true
   parent=FrontContent
   widget=Image_Particles_Glow
   properties {
    visibility=Collapsed
    brush {
     bIsDynamicallyLoaded=false
     drawAs=Image
     imageType=NoImage
     mirroring=NoMirror
     resourceName=None
     resourceObject=/Game/Athena/HUD/Creative/EOMScreens/MaterialInstances/MI_UI_Creative_EOMScreens_Shards_ParticlesTexture_1_Glow
     tiling=NoTile
     imageSize=<512 512>
     margin=<ltrb 0 0 0 0>
     outlineSettings {
      bUseBrushTransparency=false
      roundingType=HalfHeightRadius
      width=0
      color {colorUseRule=UseColor_Specified, specifiedColor=<rgba 0 0 0 0>}
      cornerRadii=<0 0 0 0>
     }
     tintColor {colorUseRule=UseColor_Specified, specifiedColor=<rgba 1 1 1 1>}
     uVRegion {bIsValid=false, max=<0 0>, min=<0 0>}
    }
   }
   slot {class=OverlaySlot, properties {horizontalAlignment=HAlign_Fill, verticalAlignment=VAlign_Fill}}
  }
  Image_Particles {
   widgetClassPath=&Image
   bIsVariable=true
   parent=FrontContent
   widget=Image_Particles
   properties {
    visibility=Collapsed
    brush {
     bIsDynamicallyLoaded=false
     drawAs=Image
     imageType=NoImage
     mirroring=NoMirror
     resourceName=None
     resourceObject=/Game/Athena/HUD/Creative/EOMScreens/MaterialInstances/MI_UI_Creative_EOMScreens_Shards_ParticlesTexture_1
     tiling=NoTile
     imageSize=<512 512>
     margin=<ltrb 0 0 0 0>
     outlineSettings {
      bUseBrushTransparency=false
      roundingType=HalfHeightRadius
      width=0
      color {colorUseRule=UseColor_Specified, specifiedColor=<rgba 0 0 0 0>}
      cornerRadii=<0 0 0 0>
     }
     tintColor {colorUseRule=UseColor_Specified, specifiedColor=<rgba 1 1 1 1>}
     uVRegion {bIsValid=false, max=<0 0>, min=<0 0>}
    }
   }
   slot {class=OverlaySlot, properties {horizontalAlignment=HAlign_Fill, verticalAlignment=VAlign_Fill}}
  }
 ]
}
