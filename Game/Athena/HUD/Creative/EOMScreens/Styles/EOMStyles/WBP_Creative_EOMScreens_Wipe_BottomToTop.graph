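# End-of-match screen style widget for Creative that plays a bottom-to-top background wipe, derived from WBP_Creative_EOMScreens_AnimationBase.
# Its single background image uses a greyscale slide material, with animations driving material progress, rotation and render opacity.
# Play EOM Animation calls the parent with color set enabled, stops running animations, plays the slide-in and starts a one-shot timer.
# That timer fires Play Ending, which plays the ending animation and clears the handle.
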WidgetBlueprint /Game/Athena/HUD/Creative/EOMScreens/Styles/EOMStyles/WBP_Creative_EOMScreens_Wipe_BottomToTop
animations [
 Anim_Slide_Bottom_To_Top {
  binding_count=1
  length_seconds=0.9
  object=&self:Anim_Slide_Bottom_To_Top
  sequence {bindings [Image_Background {
     id=#1
     possessed_class=Image
     tracks [
      {
       type=MovieSceneWidgetMaterialTrack
       sections [{
         type=MovieSceneParameterSection
         end_frame=18
         start_frame=0
         channels [
          Background Progress {type=MovieSceneScriptingFloatChannel, keys [[8,0.8484],[17,0]]}
          Background Rotation {type=MovieSceneScriptingFloatChannel, const [0.75,1]}
         ]
        }]
      }
      {
       type=MovieSceneFloatTrack
       display_name=Render Opacity
       sections [{
         type=MovieSceneFloatSection
         end_frame=18
         start_frame=0
         channels [None {type=MovieSceneScriptingFloatChannel, keys [[0,0],[8,0.95],[17,0.9]]}]
        }]
      }
     ]
    }]}
 }
 Anim_Ending {
  binding_count=1
  length_seconds=5
  object=&self:Anim_Ending
  sequence {bindings [Image_Background {
     id=#2
     possessed_class=Image
     tracks [
      {
       type=MovieSceneWidgetMaterialTrack
       sections [{
         type=MovieSceneParameterSection
         end_frame=17
         start_frame=0
         channels [
          Background Progress {type=MovieSceneScriptingFloatChannel, keys [[0,0.15],[17,0.3]]}
          Background Rotation {type=MovieSceneScriptingFloatChannel, const [0.76,1]}
          Background Smooth Progress Min {type=MovieSceneScriptingFloatChannel, const [0.52,1]}
          Background Smooth Progress Max {type=MovieSceneScriptingFloatChannel, const [1,1]}
         ]
        }]
      }
      {
       type=MovieSceneFloatTrack
       display_name=Render Opacity
       sections [{type=MovieSceneFloatSection, end_frame=17, start_frame=0, channels [None {type=MovieSceneScriptingFloatChannel, keys [[0,0],[17,0.75]]}]}]
      }
     ]
    }]}
 }
]
graphs [EventGraph {nodes [
   n1 = Event Play Ending() {is=title K2Node_CustomEvent}
   > PlayAnimationForward(self=Self-Reference(), InAnimation=Get Anim_Ending() -> (Anim_Ending:Widget Animation Object Reference))
   > Clear and Invalidate Timer by Handle(Handle=Get AnimationDelay() -> (AnimationDelay:Timer Handle Structure))
   n2 = Event Play EOM Animation!() -> (OutputDelegate, then, bUseColorSet:bool, InColorSet:E_Creative_EOMScreens_ColorSet Enum=NewEnumerator0) {is=title K2Node_Event}
   > Parent: Play EOM Animation!(execute, bUseColorSet:bool=true, InColorSet:E_Creative_EOMScreens_ColorSet Enum=@n2.InColorSet) -> (then) {is=title K2Node_CallParentFunction}
   > StopAllAnimations(self:User Widget Object Reference)
   > Clear and Invalidate Timer by Handle(Handle=Get AnimationDelay())
   > PlayAnimationForward(self=Self-Reference(), InAnimation=Get Anim_Slide_Bottom_To_Top() -> (Anim_Slide_Bottom_To_Top:Widget Animation Object Reference))
   > n3 = Set Timer by Event(Delegate=@n1, Time=2.4, bLooping=false)
   > Set AnimationDelay(AnimationDelay:Timer Handle Structure=@n3) -> (Output_Get:Timer Handle Structure)
  ]}]
variables [AnimationDelay {type=TimerHandle}]
widgets {
 info {
  inheritedWidgetCount=0
  namedSlotCount=1
  parentClass=&/WBP_Creative_EOMScreens_AnimationBase.WBP_Creative_EOMScreens_AnimationBase_C
  rootWidgetClass=null
  widgetCount=1
 }
 widgets [Image_Background {
   widgetClassPath=&Image
   bIsVariable=true
   widget=Image_Background
   properties {brush {
     bIsDynamicallyLoaded=false
     drawAs=Image
     imageType=NoImage
     mirroring=NoMirror
     resourceName=None
     resourceObject=/Game/Athena/HUD/Creative/EOMScreens/MaterialInstances/M_UI_Creative_Greyscale_Slide_Background
     tiling=NoTile
     imageSize=<32 32>
     margin=<ltrb 0 0 0 0>
     outlineSettings {
      bUseBrushTransparency=false
      roundingType=HalfHeightRadius
      width=0
      color {colorUseRule=UseColor_Specified, specifiedColor=<rgba 0 0 0 0>}
      cornerRadii=<0 0 0 0>
     }
     tintColor {colorUseRule=UseColor_Specified, specifiedColor=<rgba 1 1 1 1>}
     uVRegion {bIsValid=false, max=<0 0>, min=<0 0>}
    }}
  }]
}
